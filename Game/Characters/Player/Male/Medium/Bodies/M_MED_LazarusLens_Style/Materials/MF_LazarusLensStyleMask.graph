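# Material function that draws animated wave lines on the Lazarus Lens Style player character's face accessory.
# Four SDF texture channels are remapped with pinch and eye-weighted curves, then become time-scrolled, sharpened sine bands in two sets with separate frequencies.
# An FX texture and a UV cutoff mask the bands.
# It outputs both wave masks, their combination, and BaseColor with an eye-margin boost and two wave tints.
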MaterialFunction /Game/Characters/Player/Male/Medium/Bodies/M_MED_LazarusLens_Style/Materials/MF_LazarusLensStyleMask
graph {nodes [
  n1 = TextureSample() {
   samplerType=SAMPLERTYPE_LinearColor
   texture=/Game/Characters/Player/Male/Medium/Bodies/M_MED_LazarusLens_Style/Textures/T_LazarusLens_Style_FaceAcc_SDF
  }
  n2 = Constant() {r=0.3}
  n3 = Constant() {r=1.2}
  n4 = Constant() {r=4}
  n5 = Constant() {r=0.2}
  n6 = StaticBool() {value=true}
  n7 = ScalarParameter() {defaultValue=0.3, parameterName=WaveFrequence_Base}
  n8 = ScalarParameter() {defaultValue=0.2, parameterName=WaveFrequence_Pinch}
  n9 = Add(A=@n1.G, B=ScalarParameter() {defaultValue=0.1, parameterName=WaveFrequence_Pinch_Offset2}) {constB=0.1}
  n10 = ScalarParameter() {defaultValue=16, parameterName=WaveFrequence_Pinch_Power}
  n11 = MaterialFunctionCall(
   In=OneMinus(None=Power(Base=@n9, Exp=@n10) {constExponent=30})
   Contrast=ScalarParameter() {defaultValue=0.5, parameterName=WaveFrequence_Pinch_Contrast}
  ) {materialFunction=&CheapContrast}
  n12 = ScalarParameter() {defaultValue=6, parameterName=WaveFrequence_Eye_Power}
  n13 = MaterialFunctionCall!(
   Texture=@n1.G
   ---------------
   New Black Value=@n2
   New Middle Value=@n3
   New White Value=@n4
   Middle Point=@n5
   "--------------- "
   Define Interpolation Curve=@n6
   Interpolation Power=LinearInterpolate(
    A=LinearInterpolate(A=@n7, B=@n8, Alpha=@n11.Result) {constA=0.5}
    B=ScalarParameter() {defaultValue=1, parameterName=WaveFrequence_Eye}
    Alpha=Power(Base=@n1.G, Exp=@n12) {constExponent=30}
   ) {constA=0.5}
   Invert Interpolation Power
  ) {materialFunction=/Engine/Functions/Engine_MaterialFunctions02/3PointLevels}
  n14 = ScalarParameter() {defaultValue=200, parameterName=WaveFrequency}
  n15 = Multiply(A=Time(), B=ScalarParameter() {defaultValue=25, parameterName=WaveSpeed})
  n16 = Sine(None=Add(A=Multiply(A=@n13.Result, B=@n14) {constB=1000}, B=@n15)) {period=100}
  n17 = ScalarParameter() {defaultValue=0.18, parameterName=WaveThickness}
  n18 = Power(Base=Saturate(None=ConstantBiasScale(None=@n16)), Exp=@n17) {constExponent=0.2}
  n19 = ScalarParameter() {defaultValue=3, parameterName=WaveHardness}
  n20 = ScalarParameter() {defaultValue=50, parameterName=WaveOffset}
  n21 = ScalarParameter() {defaultValue=0.3, parameterName=WaveFrequence_Base}
  n22 = ScalarParameter() {defaultValue=0.2, parameterName=WaveFrequence_Pinch}
  n23 = Add(A=@n1.A, B=ScalarParameter() {defaultValue=0.1, parameterName=WaveFrequence_Pinch_Offset2}) {constB=0.1}
  n24 = ScalarParameter() {defaultValue=16, parameterName=WaveFrequence_Pinch_Power}
  n25 = MaterialFunctionCall(
   In=OneMinus(None=Power(Base=@n23, Exp=@n24) {constExponent=30})
   Contrast=ScalarParameter() {defaultValue=0.5, parameterName=WaveFrequence_Pinch_Contrast}
  ) {materialFunction=&CheapContrast}
  n26 = ScalarParameter() {defaultValue=6, parameterName=WaveFrequence_Eye_Power}
  n27 = MaterialFunctionCall(
   Texture=@n1.A
   New Black Value=@n2
   New Middle Value=@n3
   New White Value=@n4
   Middle Point=@n5
   Define Interpolation Curve=@n6
   Interpolation Power=LinearInterpolate(
    A=LinearInterpolate(A=@n21, B=@n22, Alpha=@n25.Result) {constA=0.5}
    B=ScalarParameter() {defaultValue=1, parameterName=WaveFrequence_Eye}
    Alpha=Power(Base=@n1.A, Exp=@n26) {constExponent=30}
   ) {constA=0.5}
  ) {materialFunction=&3PointLevels}
  n28 = Sine(None=Add(A=Add(A=@n15, B=@n20), B=Multiply(A=@n27.Result, B=@n14) {constB=1000})) {period=100}
  n29 = Power(Base=Saturate(None=ConstantBiasScale(None=@n28)), Exp=@n17) {constExponent=0.2}
  n30 = Max(
   A=OneMinus(None=MaterialFunctionCall(In=@n18, Contrast=@n19) {materialFunction=&CheapContrast}.Result)
   B=OneMinus(None=MaterialFunctionCall(In=@n29, Contrast=@n19) {materialFunction=&CheapContrast}.Result)
  )
  n31 = TextureSample() {
   samplerType=SAMPLERTYPE_LinearColor
   texture=&T_LazarusLens_Style_FaceAcc_SDF
  }
  n32 = Constant() {r=0.3}
  n33 = Constant() {r=1.2}
  n34 = Constant() {r=4}
  n35 = Constant() {r=0.2}
  n36 = StaticBool() {value=true}
  n37 = ScalarParameter() {defaultValue=0.3, parameterName=WaveFrequence_Base}
  n38 = ScalarParameter() {defaultValue=0.2, parameterName=WaveFrequence_Pinch}
  n39 = Add(A=@n31.R, B=ScalarParameter() {defaultValue=0.1, parameterName=WaveFrequence_Pinch_Offset2}) {constB=0.1}
  n40 = ScalarParameter() {defaultValue=16, parameterName=WaveFrequence_Pinch_Power}
  n41 = MaterialFunctionCall(
   In=OneMinus(None=Power(Base=@n39, Exp=@n40) {constExponent=30})
   Contrast=ScalarParameter() {defaultValue=0.5, parameterName=WaveFrequence_Pinch_Contrast}
  ) {materialFunction=&CheapContrast}
  n42 = ScalarParameter() {defaultValue=6, parameterName=WaveFrequence_Eye_Power}
  n43 = MaterialFunctionCall(
   Texture=@n31.R
   New Black Value=@n32
   New Middle Value=@n33
   New White Value=@n34
   Middle Point=@n35
   Define Interpolation Curve=@n36
   Interpolation Power=LinearInterpolate(
    A=LinearInterpolate(A=@n37, B=@n38, Alpha=@n41.Result) {constA=0.5}
    B=ScalarParameter() {defaultValue=1, parameterName=WaveFrequence_Eye}
    Alpha=Power(Base=@n31.R, Exp=@n42) {constExponent=30}
   ) {constA=0.5}
  ) {materialFunction=&3PointLevels}
  n44 = ScalarParameter() {defaultValue=400, parameterName=WaveFrequency1}
  n45 = Multiply(A=Time(), B=ScalarParameter() {defaultValue=25, parameterName=WaveSpeed})
  n46 = Sine(None=Add(A=Multiply(A=@n43.Result, B=@n44) {constB=1000}, B=@n45)) {period=100}
  n47 = ScalarParameter() {defaultValue=0.18, parameterName=WaveThickness}
  n48 = Power(Base=Saturate(None=ConstantBiasScale(None=@n46)), Exp=@n47) {constExponent=0.2}
  n49 = ScalarParameter() {defaultValue=3, parameterName=WaveHardness}
  n50 = ScalarParameter() {defaultValue=50, parameterName=WaveOffset}
  n51 = ScalarParameter() {defaultValue=0.3, parameterName=WaveFrequence_Base}
  n52 = ScalarParameter() {defaultValue=0.2, parameterName=WaveFrequence_Pinch}
  n53 = Add(A=@n31.B, B=ScalarParameter() {defaultValue=0.1, parameterName=WaveFrequence_Pinch_Offset2}) {constB=0.1}
  n54 = ScalarParameter() {defaultValue=16, parameterName=WaveFrequence_Pinch_Power}
  n55 = MaterialFunctionCall(
   In=OneMinus(None=Power(Base=@n53, Exp=@n54) {constExponent=30})
   Contrast=ScalarParameter() {defaultValue=0.5, parameterName=WaveFrequence_Pinch_Contrast}
  ) {materialFunction=&CheapContrast}
  n56 = ScalarParameter() {defaultValue=6, parameterName=WaveFrequence_Eye_Power}
  n57 = MaterialFunctionCall(
   Texture=@n31.B
   New Black Value=@n32
   New Middle Value=@n33
   New White Value=@n34
   Middle Point=@n35
   Define Interpolation Curve=@n36
   Interpolation Power=LinearInterpolate(
    A=LinearInterpolate(A=@n51, B=@n52, Alpha=@n55.Result) {constA=0.5}
    B=ScalarParameter() {defaultValue=1, parameterName=WaveFrequence_Eye}
    Alpha=Power(Base=@n31.B, Exp=@n56) {constExponent=30}
   ) {constA=0.5}
  ) {materialFunction=&3PointLevels}
  n58 = Sine(None=Add(A=Add(A=@n45, B=@n50), B=Multiply(A=@n57.Result, B=@n44) {constB=1000})) {period=100}
  n59 = Power(Base=Saturate(None=ConstantBiasScale(None=@n58)), Exp=@n47) {constExponent=0.2}
  n60 = Max(
   A=OneMinus(None=MaterialFunctionCall(In=@n48, Contrast=@n49) {materialFunction=&CheapContrast}.Result)
   B=OneMinus(None=MaterialFunctionCall(In=@n59, Contrast=@n49) {materialFunction=&CheapContrast}.Result)
  )
  n61 = TextureSample() {
   samplerType=SAMPLERTYPE_Masks
   texture=/Game/Characters/Player/Male/Medium/Bodies/M_MED_LazarusLens_Style/Textures/T_LazarusLens_Style_FaceAcc_FX
  }
  n62 = Add(A=ComponentMask(None=TextureCoordinate()) {r=true}) {constB=0.135}
  n63 = Multiply(A=OneMinus(None=Max(A=@n61.R, B=@n61.R)), B=OneMinus(None=Floor(None=@n62)))
  FunctionOutput(None=Multiply(A=Max(A=@n30, B=@n60), B=@n63)) {outputName=WavesMask}
  n64 = ScalarParameter() {defaultValue=0.4, parameterName=EyeMarginBrightnessBoost}
  n65 = Multiply(A=FunctionInput() {inputName=BaseColor}, B=Add(A=Multiply(A=@n61.G, B=@n64) {constB=0.2}))
  n66 = VectorParameter() {parameterName=WaveColor2, defaultValue=<rgba 0.0122 0.0049 0.0052 1>}
  n67 = Multiply(A=@n30, B=@n63)
  n68 = Multiply(A=@n60, B=@n63)
  FunctionOutput(None=LinearInterpolate(
    A=LinearInterpolate(A=@n65, B=@n66.RGB, Alpha=@n67) {constA=0.18}
    B=VectorParameter() {parameterName=WaveColor1, defaultValue=<rgba 0.0232 0.3613 0.2542 1>}.RGB
    Alpha=@n68
   ) {constA=0.18}) {outputName=BaseColor}
  FunctionOutput(None=@n67) {outputName=Waves1}
  FunctionOutput(None=@n68) {outputName=Waves2}
 ]}
