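# Parent material for the LazarusLens Style character's face accessory, built on the shared character parent pipeline.
# It runs the character pre-FX function, then passes the base color through a LazarusLens style mask function.
# The mask's waves output blends roughness toward a WaveRoughness parameter before the modified attributes go into the post-FX function.
# Shading model comes from the material expression, with skeletal mesh, morph target, mesh deformer and Nanite usage enabled.
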Material /Game/Characters/Player/Male/Medium/Bodies/M_MED_LazarusLens_Style/Materials/M_LazarusLens_Style_FaceAcc_Parent
graph {
 properties {
  bUseMaterialAttributes=true
  bUsedWithMeshDeformer=true
  bUsedWithMorphTargets=true
  bUsedWithNanite=true
  bUsedWithSkeletalMesh=true
  pixelDepthOffsetMode=PDOM_Legacy
  shadingModel=MSM_FromMaterialExpression
  usedShadingModels=Default Lit | Subsurface | Clear Coat | Subsurface Profile | Cloth
 }
 nodes [
  n1 = MaterialFunctionCall() {materialFunction=&MF_CharacterParent_2023_PreFX}
  n2 = GetMaterialAttributes(None=@n1.Result) {attributeGetTypes [#1, #2]}
  n3 = MaterialFunctionCall!(BaseColor=@n2.Base Color) {materialFunction=&/MF_LazarusLensStyleMask}
  n4 = ScalarParameter() {defaultValue=0.44, parameterName=WaveRoughness}
  MaterialFunctionCall(
   Material=SetMaterialAttributes!(
    MaterialAttributes=@n2.MaterialAttributes
    Base Color=@n3.BaseColor
    Roughness=LinearInterpolate(A=@n2.Roughness, B=@n4, Alpha=@n3.WavesMask)
   ) {attributeSetTypes [#1, #2]}
   M=@n1.M
   Pre Skinned Local Position=@n1.M
   Distance Blend=@n1.M
   Fresnel=@n1.M
  ) {materialFunction=&MF_CharacterParent_2023_PostFX}
 ]
}
